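; Season 7 challenge C quest interactable that rewards a tagged player with a brazier sequence and the Prisoner skin's Stage3 variant.
; On authority, capsule overlap checks the pawn's gameplay tags, sends a quest stat event, then changes the variant after delays, with transition FX on remotes.
; A looping timer lights eight braziers sequentially, then timelines flare them and collision re-enables.
; The construction script offsets each flame along its forward vector and height.
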(event EventBeginPlay
  ;[1 1 1] BEGIN PLAY: Bind event to listen for overlap
  (Collision|BindEventtoOnComponentBeginOverlap (Variables|Default|GetCentralOverlap) (EventDispatchers|CreateEvent :!event Custom|PawnTouchedCapsule)))

(event Custom|PawnTouchedCapsule
  (params (OverlappedComponent :type Primitive%20Component%20Object%20Reference) (OtherActor :type Actor%20Object%20Reference) (OtherComp :type Primitive%20Component%20Object%20Reference) (OtherBodyIndex :type Integer) (bFromSweep :type Boolean) (SweepResult :type Hit%20Result%20Structure%20%28by%20ref%29))
  (bind _returnvalue (Fortnite|GetFortPlayerControllerfromActor OtherActor))
  ;[1 1 1] On Touch, test for Season7C requirements
  (Networking|SwitchHasAuthority
    (:Authority
      (bind _returnvalue_2 (GameplayTags|GetOwnedGameplayTags OtherActor :!class /Script/GameplayTags.GameplayTagAssetInterface))
      (if (GameplayTags|HasAllTags _returnvalue_2 (Variables|Default|GetCustomTags) true)
        (Quest|SendComplexCustomStatEvent (Quests|GetQuestManager _returnvalue "Invalid") 0 "(GameplayTags=((TagName=\"Athena.Location.Meta.Season7.C\")))" "(GameplayTags=)" 1)
        (if true
          (CallFunction|ChallengeSucceeded)
          (Utilities|FlowControl|Delay (+ (* (Variables|Default|GetLightNextBrazierPauseTime) 8.0) 1.0))
          ;[1 1 1] Apply Variant in game
          (bind _asfort_pawn (Utilities|Casting|CastToFortPawn OtherActor))
          (CallFunction|HandleCharacterVariantTransition _asfort_pawn)
          (Utilities|FlowControl|Delay 0.8)
          (bind _asfort_player_controller_athena (Utilities|Casting|CastToFortPlayerControllerAthena _returnvalue))
          (Class|FortPlayerController|ChangeActiveVariantforCosmeticItem _asfort_player_controller_athena "AthenaCharacter:cid_332_athena_commando_m_prisoner" "(TagName=\"Cosmetics.Variant.Channel.Progressive\")" "(TagName=\"Cosmetics.Variant.Property.Stage3\")" 0)
          (else
            (Development|PrintString "Quest Active Fail" false false)))
        (elif (Variables|Default|GetTestModeEnabled)
          (CallFunction|ChallengeSucceeded))))))

(event Custom|ChallengeSucceeded
  ;[1 1 1] Upon challenge success, turn off collision and start timer for lighting braziers
  (bind _returnvalue (Utilities|Time|SetTimerbyEvent (delegate Custom|LIghtNextBrazier) (Variables|Default|GetLightNextBrazierPauseTime) true))
  (Variables|Default|SetLightingTimer _returnvalue)
  (Collision|SetCollisionEnabled (Variables|Default|GetCentralOverlap) "NoCollision"))

(event Custom|LIghtNextBrazier
  (bind _braziernumber (Variables|Default|GetBrazierNumber))
  (bind _particlesystem78 (Variables|Default|GetParticleSystem78))
  (bind _particlesystem77 (Variables|Default|GetParticleSystem77))
  (bind _particlesystem76 (Variables|Default|GetParticleSystem76))
  (bind _particlesystem75 (Variables|Default|GetParticleSystem75))
  (bind _particlesystem74 (Variables|Default|GetParticleSystem74))
  (bind _particlesystem73 (Variables|Default|GetParticleSystem73))
  (bind _particlesystem72 (Variables|Default|GetParticleSystem72))
  (bind _particlesystem71 (Variables|Default|GetParticleSystem71))
  (bind _returnvalue (Utilities|Select :Index _braziernumber :Option%200 _particlesystem71 :Option%201 _particlesystem72 :Option%202 _particlesystem73 :Option%203 _particlesystem74 :Option%204 _particlesystem75 :Option%205 _particlesystem76 :Option%206 _particlesystem77 :Option%207 _particlesystem78))
  (bind _fireemitters (Variables|Default|GetFireEmitters))
  ;[1 1 1] Light up Braziers one at a time
  (Components|Activation|Activate _returnvalue false)
  (CallFunction|FlameIgnite _returnvalue)
  (Audio|PlaySoundatLocation (Variables|Default|GetLightBrazierSound) (Transformation|GetWorldLocation _returnvalue))
  (bind _result (Math|Integer|IncrementInt (Variables|Default|GetBrazierNumber))
    (:%20%20
      (if (>= _result 8)
        (Utilities|Time|ClearandInvalidateTimerbyHandle (Variables|Default|GetLightingTimer))
        (Utilities|FlowControl|Delay (Variables|Default|GetStayLitTime))
        ;[1 1 1] Flares Flame Mesh before Deactivation
        (bind (_direction _flame_size__normalized) (Timeline|Fade%20Flare PlayFromStart)
          (:Update
            (for _array_element _fireemitters
              (Effects|Components|ParticleSystem|SetFloatParameter _array_element "FlameMeshSize" _flame_size__normalized)
              (Effects|Components|ParticleSystem|SetVectorParameter _array_element "Brightness" (Math|Conversions|ToVector(Float) _flame_size__normalized))))
          (:Finished
            ;[1 1 1] Extinguish Braziers
            (Components|Activation|Deactivate (Variables|Default|GetParticleSystem71))
            (Variables|Default|SetBrazierNumber 0)
            ;[1 1 1] Wait and then Turn Collision back on and reset
            (Utilities|FlowControl|Delay (Math|Float|SelectFloat 5.0 1.0 (Variables|Default|GetTestModeEnabled)))
            (Collision|SetCollisionEnabled (Variables|Default|GetCentralOverlap) "QueryAndPhysics"))
          (:BurstEvent
            (for _array_element_2 _fireemitters
              (Effects|Components|ParticleSystem|SetFloatParameter _array_element_2 "BurstFlames" 1.0))))))))

(event Custom|Flame%20Ignite
  (params (To%20Ignite :type Cascade%20Particle%20System%20Component%20Object%20Reference))
  ;[1 1 1] Start Up Brazier FX
  (bind (_direction _flame_size__normalized) (Timeline|Ignite%20Flame PlayFromStart)
    (:Update
      (Effects|Components|ParticleSystem|SetFloatParameter To%20Ignite "FlameMeshSize" _flame_size__normalized)
      (Effects|Components|ParticleSystem|SetVectorParameter To%20Ignite "Brightness" (Math|Conversions|ToVector(Float) _flame_size__normalized)))
    (:Finished
      (Effects|Components|ParticleSystem|SetFloatParameter To%20Ignite "FlameMeshSize" 0.0)
      (Effects|Components|ParticleSystem|SetFloatParameter To%20Ignite "BurstFlames" 0.0)
      (Effects|Components|ParticleSystem|SetVectorParameter To%20Ignite "Brightness" "0, 0, 0"))
    (:BurstEvent
      (Effects|Components|ParticleSystem|SetFloatParameter To%20Ignite "BurstFlames" 1.0))))

(event Custom|HandleCharacterVariantTransition
  (params (TransitioningPawn :type Fort%20Pawn%20Object%20Reference))
  (bind _returnvalue (Transformation|GetActorLocation TransitioningPawn))
  (bind _mesh (Variables|/Script/Engine.Character|GetMesh TransitioningPawn))
  ;[1 1 1] Handle Season7C Transition FX and sound
  (Networking|SwitchHasAuthority
    (:Remote
      (Utilities|IsValid TransitioningPawn
        (:Is%20Valid
          (Utilities|IsValid _mesh
            (:Is%20Valid
              (Effects|SpawnEmitterAttached :EmitterTemplate "/Figment_S07/Characters/Player/Male/Medium/Bodies/M_MED_Prisoner/FX/P_Prisoner_Transition.P_Prisoner_Transition" :AttachToComponent _mesh :AttachPointName "spine_05" :Scale "1.000000,1.000000,1.000000" :LocationType "SnapToTargetIncludingScale" :bAutoDestroy true :PoolingMethod "None" :bAutoActivate true)
              (Audio|PlaySoundatLocation (Variables|Default|GetSuccessSound3rdPerson) _returnvalue)
              (Audio|PlaySoundatLocation (Variables|Default|GetSuccessSound) _returnvalue)
              (Camera|PlayWorldCameraShake "/FNGameplayCues/Effects/Fort_Effects/Camera_Effects/B_s8Leadup1_CameraShake_01.B_s8Leadup1_CameraShake_01_C" _returnvalue 2e+03 1.4e+04 2.0))))))))

(fn ConstructionScript
  (bind _output_get (Variables|Default|SetFireEmitters (Utilities|Array|MakeArray (Variables|Default|GetParticleSystem71) (Variables|Default|GetParticleSystem72) (Variables|Default|GetParticleSystem73) (Variables|Default|GetParticleSystem74) (Variables|Default|GetParticleSystem75) (Variables|Default|GetParticleSystem78) (Variables|Default|GetParticleSystem77) (Variables|Default|GetParticleSystem76))))
  (for _array_element _output_get
    (Transformation|SetRelativeLocation _array_element (Math|Vector|MakeVector (.x (* (Transformation|GetForwardVector _array_element) (Variables|Default|GetFlameRadiusOffset))) (.y (* (Transformation|GetForwardVector _array_element) (Variables|Default|GetFlameRadiusOffset))) (+ (.z (* (Transformation|GetForwardVector _array_element) (Variables|Default|GetFlameRadiusOffset))) (Variables|Default|GetFlameHeightOffset))) false false)))
