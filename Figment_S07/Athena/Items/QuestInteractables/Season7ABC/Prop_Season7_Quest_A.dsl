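; Season 7 quest prop: a key that, once its objective completes, switches the Prisoner outfit to its Stage1 progressive variant.
; For a local pawn with the required cosmetic tag, sphere overlaps drive timelines that lift and spin the key, then return it, with audio.
; Quest updates unbind those overlaps.
; Completion shrinks the key, while the server event spawns transition FX on the pawn, then changes the variant after a short delay.
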(event EventBeginPlay
  (bind _returnvalue (Fortnite|GetLocalFortPlayerControllers true))
  (bind _output (Utilities|Array|Get(acopy) _returnvalue))
  (bind _returnvalue_2 (Game|Player|GetPlayerPawn _output))
  (bind _sm_key (Variables|Default|GetSM_Key))
  ;[1 1 1] Init transform vals, sounds and fx
  (|CallParentFunction)
  (Variables|Default|SetInitKeyLoc (Transformation|GetWorldLocation _sm_key))
  (Variables|Default|SetInitKeyRot (Transformation|GetWorldRotation _sm_key))
  (Variables|Default|SetSuccessFX 0)
  (Variables|Default|SetSuccessSound (Variables|Default|GetNewSuccessSound))
  ;[1 1 1] Bind and Unbind floating when close
  (Utilities|IsValid _output
    (:Is%20Valid
      (Utilities|IsValid _returnvalue_2
        (:Is%20Valid
          (bind _returnvalue_3 (GameplayTags|GetOwnedGameplayTags _returnvalue_2 :!class /Script/GameplayTags.GameplayTagAssetInterface))
          (if (GameplayTags|HasAllTags _returnvalue_3 (Variables|Default|GetCosmeticCustomTag) true)
            (Collision|BindEventtoOnComponentBeginOverlap (Variables|Default|GetOverlapSphereEnter) (EventDispatchers|CreateEvent :!event Custom|EnteredVicinity))
            (Collision|BindEventtoOnComponentEndOverlap (Variables|Default|GetOverlapSphereExit) (EventDispatchers|CreateEvent :!event Custom|ExitedVicinity))))))))

(event EventQuestInitializedAndValid
  (bind _returnvalue (Fortnite|GetLocalFortPlayerControllers true))
  (bind _output (Utilities|Array|Get(acopy) _returnvalue))
  (bind _returnvalue_2 (Game|Player|GetPlayerPawn _output))
  ;[1 1 1] Bind and Unbind floating when close
  (|CallParentFunction)
  (Utilities|IsValid _output
    (:Is%20Valid
      (Utilities|IsValid _returnvalue_2
        (:Is%20Valid
          (bind _returnvalue_3 (GameplayTags|GetOwnedGameplayTags _returnvalue_2 :!class /Script/GameplayTags.GameplayTagAssetInterface))
          (if (GameplayTags|HasAllTags _returnvalue_3 (Variables|Default|GetCosmeticCustomTag) true)
            (Collision|BindEventtoOnComponentBeginOverlap (Variables|Default|GetOverlapSphereEnter) (EventDispatchers|CreateEvent :!event Custom|EnteredVicinity))
            (Collision|BindEventtoOnComponentEndOverlap (Variables|Default|GetOverlapSphereExit) (EventDispatchers|CreateEvent :!event Custom|ExitedVicinity))))))))

(event EventQuestUpdatedFromAnywhere
  ;[1 1 1] Bind and Unbind floating when close
  (|CallParentFunction)
  (Collision|UnbindEventfromOnComponentBeginOverlap (Variables|Default|GetOverlapSphereEnter) (EventDispatchers|CreateEvent :!event Custom|EnteredVicinity))
  (Collision|UnbindEventfromOnComponentEndOverlap (Variables|Default|GetOverlapSphereExit) (EventDispatchers|CreateEvent :!event Custom|ExitedVicinity)))

(event Custom|EnteredVicinity
  (params (OverlappedComponent :type Primitive%20Component%20Object%20Reference) (OtherActor :type Actor%20Object%20Reference) (OtherComp :type Primitive%20Component%20Object%20Reference) (OtherBodyIndex :type Integer) (bFromSweep :type Boolean) (SweepResult :type Hit%20Result%20Structure%20%28by%20ref%29))
  ;[1 1 1] Floating Movement
  (Utilities|FlowControl|Sequence
    (:then_0
      (bind (_direction _lift_time) (Timeline|Lift%20Up PlayFromStart)
        (:Update
          (Transformation|SetWorldLocation (Variables|Default|GetSM_Key) (Math|Vector|Lerp(Vector) (Variables|Default|GetInitKeyLoc) (+ (Variables|Default|GetInitKeyLoc) "0, 0,40.000000") _lift_time) false false))))
    (:then_1
      (Components|Timeline|SetLooping (Variables|Prop_Season7_Quest_A|GetSpin) true)
      (bind (_direction_2 _rotate_time) (Timeline|Spin Play)
        (:Update
          (Transformation|SetWorldRotation (Variables|Default|GetSM_Key) (Math|Rotator|Lerp(Rotator) (Variables|Default|GetInitKeyRot) (Math|Rotator|MakeRotator (.roll (Variables|Default|GetInitKeyRot)) (+ (.pitch (Variables|Default|GetInitKeyRot)) 3.6e+02) (+ (.yaw (Variables|Default|GetInitKeyRot)) 3.6e+02)) _rotate_time) false false))))
    (:then_2
      (if (not (Variables|Default|GetIs%20Complete))
        (Audio|Components|Audio|Play (Variables|Default|GetAudio_Key_Float) 0.0)))))

(event Custom|ExitedVicinity
  (params (OverlappedComponent :type Primitive%20Component%20Object%20Reference) (OtherActor :type Actor%20Object%20Reference) (OtherComp :type Primitive%20Component%20Object%20Reference) (OtherBodyIndex :type Integer))
  (bind _sm_key (Variables|Default|GetSM_Key))
  ;[1 1 1] End Floating Movement
  (Utilities|FlowControl|Sequence
    (:then_0
      ;[1 1 1] Floating Movement
      (Timeline|Spin Stop))
    (:then_1
      (Variables|Default|SetCurrentKeyLoc (Transformation|GetWorldLocation _sm_key))
      (Variables|Default|SetCurrentKeyRot (Transformation|GetWorldRotation _sm_key))
      (bind (_direction _lift_time) (Timeline|Timeline_0 PlayFromStart)
        (:Update
          (Transformation|SetWorldLocationAndRotation (Variables|Default|GetSM_Key) (Math|Vector|Lerp(Vector) (Variables|Default|GetCurrentKeyLoc) (Variables|Default|GetInitKeyLoc) _lift_time) (Math|Rotator|Lerp(Rotator) (Variables|Default|GetCurrentKeyRot) (Variables|Default|GetInitKeyRot) _lift_time) false false))))
    (:then_2
      (Audio|Components|Audio|FadeOut (Variables|Default|GetAudio_Key_Float) 0.25 0.0 "Linear" :!class /Script/Engine.AudioComponent))
    (:then_3
      (Audio|Components|Audio|Play (Variables|Default|GetAudio_Key_Drop) 0.0))))

(event EventThisQuestObjectiveComplete
  (params (IneractingPawn :type Fort%20Pawn%20Object%20Reference))
  ;[1 1 1] Success Movement
  (|CallParentFunction IneractingPawn)
  (bind (_direction _pop_time) (Timeline|FinishQuestPop PlayFromStart)
    (:Update
      (Transformation|SetRelativeScale3D (Variables|Default|GetSM_Key) (Math|Vector|Lerp(Vector) "1.000000,1.000000,1.000000" "0.001000,0.001000,0.001000" _pop_time)))))

(event EventServer_QuestObjectiveCompleted
  (params (InteractingPawn :type Fort%20Pawn%20Object%20Reference))
  ;[1 1 1] Update skin
  (|CallParentFunction InteractingPawn)
  (CallFunction|HandleCharacterVariantTransition InteractingPawn)
  (Utilities|FlowControl|Delay 0.5)
  (bind _asathena_player_controller (Utilities|Casting|CastToAthena_PlayerController (Pawn|GetController InteractingPawn)))
  (Class|FortPlayerController|ChangeActiveVariantforCosmeticItem _asathena_player_controller "AthenaCharacter:cid_332_athena_commando_m_prisoner" "(TagName=\"Cosmetics.Variant.Channel.Progressive\")" "(TagName=\"Cosmetics.Variant.Property.Stage1\")" 0))

(event Custom|HandleCharacterVariantTransition
  (params (TransitioningPawn :type Fort%20Pawn%20Object%20Reference))
  (bind _mesh (Variables|/Script/Engine.Character|GetMesh TransitioningPawn))
  ;[1 1 1] Play Update FX
  (Utilities|IsValid TransitioningPawn
    (:Is%20Valid
      (Utilities|IsValid _mesh
        (:Is%20Valid
          (Effects|SpawnEmitterAttached :EmitterTemplate "/Figment_S07/Characters/Player/Male/Medium/Bodies/M_MED_Prisoner/FX/P_Prisoner_Transition.P_Prisoner_Transition" :AttachToComponent _mesh :AttachPointName "spine_05" :Scale "1.000000,1.000000,1.000000" :LocationType "SnapToTargetIncludingScale" :bAutoDestroy true :PoolingMethod "None" :bAutoActivate true)
          (Audio|Components|Audio|FadeOut (Variables|Default|GetAudio_Key_Float) 0.25 0.0 "Linear" :!class /Script/Engine.AudioComponent)
          (Variables|Default|SetIs%20Complete true))))))

(fn ConstructionScript
  (|CallParentFunction))
